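# Physics asset for the TennisLeash backpack cosmetic, giving secondary motion to its dangling grenades, socks, candy pieces and zippers.
# Capsule bodies hang mostly from dyn_base_joint; the two grenades carry their own rings, bows and fur pieces.
# Two three-joint sock chains start separately at the base and pack bodies.
# Angular limits let grenades and zippers swing widely, keep candy stiff, lock fur rotation and restrict some bows and rings to fewer axes.
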PhysicsAsset /BRCosmetics/Accessories/FORT_Backpacks/Backpack_M_MED_TennisLeash/Meshes/M_MED_TennisLeash_Pack_Physics
bodies {
 dyn_backpack_joint [{shapeType=Capsule, length=14.505, radius=12.505, shapeName=dyn_backpack_joint_capsule, rotation=<rot 0 -90 0>}]
 dyn_base_joint [{
   shapeType=Capsule
   length=5.3001
   radius=2.3221
   shapeName=dyn_base_joint_capsule
   center=<0.0818 -0.6198 -0.5563>
   rotation=<rot -0.0847 -89.6527 -0.9547>
  }]
 dyn_bow_1_joint [{
   shapeType=Capsule
   length=2.6977
   radius=1
   shapeName=dyn_bow_1_joint_capsule
   center=<2.3788 -0.2372 -0.0773>
   rotation=<rot 0 -89.7513 -88.6706>
  }]
 dyn_bow_2_joint [{
   shapeType=Capsule
   length=1.4634
   radius=1
   shapeName=dyn_bow_2_joint_capsule
   center=<1.9663 -0.4959 0.244>
   rotation=<rot 1.961 78.9687 -86.6005>
  }]
 dyn_bow_3_joint [{
   shapeType=Capsule
   length=1.939
   radius=1
   shapeName=dyn_bow_3_joint_capsule
   center=<1.6228 -0.268 -0.0174>
   rotation=<rot 0.0002 -85.9471 -88.5429>
  }]
 dyn_bow_4_joint [{
   shapeType=Capsule
   length=1.6226
   radius=1
   shapeName=dyn_bow_4_joint_capsule
   center=<2.0793 -0.0151 -0.2231>
   rotation=<rot -0.6326 85.3558 -88.2614>
  }]
 dyn_candy_1_joint [{
   shapeType=Capsule
   length=17.4555
   radius=0.5
   shapeName=dyn_candy_1_joint_capsule
   center=<-0.2091 -0.138 -3.4245>
   rotation=<rot -0.5997 -90.0389 -0.4439>
  }]
 dyn_candy_2_joint [{
   shapeType=Capsule
   length=19.5287
   radius=0.5
   shapeName=dyn_candy_2_joint_capsule
   center=<-0.1303 -0.1546 -5.2837>
   rotation=<rot 0.6845 -90.0297 -0.3397>
  }]
 dyn_fur_1_joint [{
   shapeType=Capsule
   length=2.1083
   radius=3.5
   shapeName=dyn_fur_1_joint_capsule
   center=<0.0065 -0.4507 -0.5907>
   rotation=<rot 0 178.4247 -179.3755>
  }]
 dyn_fur_2_joint [{
   shapeType=Capsule
   length=3.1295
   radius=3
   shapeName=dyn_fur_2_joint_capsule
   center=<-0.0577 0.1154 -0.05>
   rotation=<rot 0 5.6729 -81.5368>
  }]
 dyn_nade_1_joint [{shapeType=Capsule, length=12.5, radius=5.5, shapeName=dyn_nade_1_joint_capsule, center=<-23.5964 -0.0002 0>, rotation=<rot 90 -90 -90.0006>}]
 dyn_nade_2_joint [{shapeType=Capsule, length=12.5, radius=5, shapeName=dyn_nade_2_joint_capsule, center=<-22.9908 -0.0002 0>, rotation=<rot -90 180 -179.9996>}]
 dyn_nade_ring_1_joint [{
   shapeType=Capsule
   length=3.5349
   radius=0.5
   shapeName=dyn_nade_ring_1_joint_capsule
   center=<2.1647 0.0054 -0.0014>
   rotation=<rot -0.7056 -89.9998 -86.3145>
  }]
 dyn_nade_ring_2_joint [{
   shapeType=Capsule
   length=3.0729
   radius=0.5
   shapeName=dyn_nade_ring_2_joint_capsule
   center=<2.13 -0.0042 0.0076>
   rotation=<rot 0.0002 -90.1739 -87.3778>
  }]
 dyn_sock_1_joint [{
   shapeType=Capsule
   length=1.9792
   radius=1
   shapeName=dyn_sock_1_joint_capsule
   center=<-0.0657 3.2014 0.8351>
   rotation=<rot 54.0214 -26.0343 60.6696>
  }]
 dyn_sock_2_joint [{
   shapeType=Capsule
   length=2.2713
   radius=1.5
   shapeName=dyn_sock_2_joint_capsule
   center=<2.5468 -0.1013 -0.0393>
   rotation=<rot -21.425 -89.9998 84.2475>
  }]
 dyn_sock_3_joint [{
   shapeType=Capsule
   length=2
   radius=1.5
   shapeName=dyn_sock_3_joint_capsule
   center=<2.714 0.0625 0.815>
   rotation=<rot -3.3982 -89.9999 80.9436>
  }]
 dyn_sock_4_joint [{
   shapeType=Capsule
   length=3.5
   radius=1.5
   shapeName=dyn_sock_4_joint_capsule
   center=<-0.0417 -3.4364 1.0913>
   rotation=<rot -3.4036 -5.5917 -70.2804>
  }]
 dyn_sock_5_joint [{
   shapeType=Capsule
   length=1.1899
   radius=1.9324
   shapeName=dyn_sock_5_joint_capsule
   center=<3.0395 -0.432 -0.0174>
   rotation=<rot 5.8319 -95.0258 -91.1393>
  }]
 dyn_sock_6_joint [{
   shapeType=Capsule
   length=1.5
   radius=2.1624
   shapeName=dyn_sock_6_joint_capsule
   center=<3.2702 -1.005 0.6302>
   rotation=<rot 10.8836 -110.376 -95.5166>
  }]
 dyn_zip_1_joint [{
   shapeType=Capsule
   length=0.505
   radius=0.505
   shapeName=dyn_zip_1_joint_capsule
   center=<1.1885 -0.0188 -0.0008>
   rotation=<rot 0 -90.8865 -89.9611>
  }]
 dyn_zip_2_joint [{
   shapeType=Capsule
   length=1.01
   radius=0.5
   shapeName=dyn_zip_2_joint_capsule
   center=<1.1894 -0.0193 -0.0026>
   rotation=<rot 0 -90.7466 -89.9001>
  }]
 pack [{shapeType=Capsule, length=0.505, radius=0.505, shapeName=pack_capsule, rotation=<rot 0 -90 0>}]
}
constraints [
 {bone1Name=dyn_sock_1_joint, bone2Name=dyn_base_joint, swing1LimitDegrees=15, swing2LimitDegrees=15, twistLimitDegrees=1}
 {
  bone1Name=dyn_fur_1_joint
  bone2Name=dyn_nade_1_joint
  swing1LimitDegrees=5
  swing1Motion=Locked
  swing2LimitDegrees=5
  swing2Motion=Locked
  twistLimitDegrees=0
  twistMotion=Locked
 }
 {bone1Name=dyn_bow_1_joint, bone2Name=dyn_nade_1_joint, swing1LimitDegrees=10, swing2LimitDegrees=20, twistLimitDegrees=1}
 {
  bone1Name=dyn_bow_2_joint
  bone2Name=dyn_nade_1_joint
  swing1LimitDegrees=5
  swing1Motion=Locked
  swing2LimitDegrees=5
  twistLimitDegrees=0
  twistMotion=Locked
 }
 {
  bone1Name=dyn_nade_ring_1_joint
  bone2Name=dyn_nade_1_joint
  swing1LimitDegrees=5
  swing1Motion=Locked
  swing2LimitDegrees=15
  twistLimitDegrees=0
  twistMotion=Locked
 }
 {
  bone1Name=dyn_nade_ring_2_joint
  bone2Name=dyn_nade_2_joint
  swing1LimitDegrees=1
  swing2LimitDegrees=15
  twistLimitDegrees=0
  twistMotion=Locked
 }
 {
  bone1Name=dyn_fur_2_joint
  bone2Name=dyn_nade_2_joint
  swing1LimitDegrees=5
  swing1Motion=Locked
  swing2LimitDegrees=5
  swing2Motion=Locked
  twistLimitDegrees=0
  twistMotion=Locked
 }
 {
  bone1Name=dyn_bow_3_joint
  bone2Name=dyn_nade_2_joint
  swing1LimitDegrees=5
  swing1Motion=Locked
  swing2LimitDegrees=5
  twistLimitDegrees=0
  twistMotion=Locked
 }
 {bone1Name=dyn_bow_4_joint, bone2Name=dyn_nade_2_joint, swing1LimitDegrees=20, swing2LimitDegrees=10, twistLimitDegrees=1}
 {bone1Name=dyn_sock_2_joint, bone2Name=dyn_sock_1_joint, swing1LimitDegrees=15, swing2LimitDegrees=15, twistLimitDegrees=1}
 {bone1Name=dyn_sock_3_joint, bone2Name=dyn_sock_2_joint, swing1LimitDegrees=10, swing2LimitDegrees=20, twistLimitDegrees=1}
 {bone1Name=dyn_zip_2_joint, bone2Name=dyn_base_joint, swing1LimitDegrees=70, swing2LimitDegrees=15, twistLimitDegrees=3}
 {bone1Name=dyn_zip_1_joint, bone2Name=dyn_base_joint, swing1LimitDegrees=90, swing2LimitDegrees=20, twistLimitDegrees=1}
 {bone1Name=dyn_sock_6_joint, bone2Name=dyn_sock_5_joint, swing1LimitDegrees=10, swing2LimitDegrees=15, twistLimitDegrees=1}
 {bone1Name=dyn_sock_5_joint, bone2Name=dyn_sock_4_joint, swing1LimitDegrees=15, swing2LimitDegrees=15, twistLimitDegrees=1}
 {bone1Name=dyn_sock_4_joint, bone2Name=pack, swing1LimitDegrees=15, swing2LimitDegrees=10, twistLimitDegrees=1}
 {bone1Name=dyn_candy_1_joint, bone2Name=dyn_base_joint, swing1LimitDegrees=5, swing2LimitDegrees=5, twistLimitDegrees=0}
 {bone1Name=dyn_candy_2_joint, bone2Name=dyn_base_joint, swing1LimitDegrees=5, swing2LimitDegrees=5, twistLimitDegrees=0}
 {bone1Name=dyn_nade_1_joint, bone2Name=dyn_base_joint, swing1LimitDegrees=30, swing2LimitDegrees=40, twistLimitDegrees=1}
 {bone1Name=dyn_nade_2_joint, bone2Name=dyn_base_joint, swing1LimitDegrees=30, swing2LimitDegrees=40, twistLimitDegrees=1}
]
