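# Animation Blueprint for the TennisLeash backpack cosmetic, anchoring the pack to the character's pelvis and simulating its physics.
# Each update caches the attach parent's pelvis world transform, which replaces dyn_backpack_joint's transform in world space.
# While bIsSkydiveFloating is true, additive bone offsets adjust the base and two sock joints.
# A RigidBody node then simulates the result using the pack's physics asset, with clamped simulation-space motion.
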AnimBlueprint /BRCosmetics/Accessories/FORT_Backpacks/Backpack_M_MED_TennisLeash/Meshes/M_MED_TennisLeash_Pack_AnimBP
target_skeleton=&/M_MED_TennisLeash_Pack_Skeleton
graphs [
 AnimGraph {nodes [
   n1 = Get!() -> (Transform Pelvis_Location:Vector, Transform Pelvis_Rotation:Rotator, Transform Pelvis_Scale:Vector)
   n2 = "Transform (Modify) Bone - Bone: dyn_backpack_joint"(
    ComponentPose=Local To Component(LocalPose=Input Pose())
    Alpha=1
    Translation=@n1
    Rotation=@n1.Transform Pelvis_Rotation
    Scale=@n1.Transform Pelvis_Scale
   ) {
    rotationMode=BMM_Replace
    rotationSpace=BCS_WorldSpace
    scaleMode=BMM_Replace
    scaleSpace=BCS_WorldSpace
    translationMode=BMM_Replace
    translationSpace=BCS_WorldSpace
    alphaBoolBlend {blendOption=Linear}
    boneToModify {boneName=dyn_backpack_joint}
   }
   n3 = Get bIsSkydiveFloating()
   n4 = "Transform (Modify) Bone - Bone: dyn_base_joint"!(ComponentPose=@n2, bAlphaBoolEnabled:bool=@n3, Translation="0,-1.5,4", Rotation="0,0,20", Scale="1,1,1") -> (Pose) {
    alphaInputType=Bool
    rotationMode=BMM_Additive
    translationMode=BMM_Additive
    alphaBoolBlend {blendOption=Linear}
    boneToModify {boneName=dyn_base_joint}
    rotation {roll=20}
    translation {y=-1.5, z=4}
   }
   n5 = "Transform (Modify) Bone - Bone: dyn_sock_4_joint"!(ComponentPose=@n4, bAlphaBoolEnabled:bool=@n3, Translation="0,0,0", Rotation="0,0,30", Scale="1,1,1") -> (Pose) {
    alphaInputType=Bool
    rotationMode=BMM_Additive
    alphaBoolBlend {blendOption=Linear}
    boneToModify {boneName=dyn_sock_4_joint}
    rotation {roll=30}
   }
   n6 = "Transform (Modify) Bone - Bone: dyn_sock_1_joint"!(ComponentPose=@n5, bAlphaBoolEnabled:bool=@n3, Translation="0,0,0", Rotation="0,0,30", Scale="1,1,1") -> (Pose) {
    alphaInputType=Bool
    rotationMode=BMM_Additive
    alphaBoolBlend {blendOption=Linear}
    boneToModify {boneName=dyn_sock_1_joint}
    rotation {roll=30}
   }
   n7 = RigidBody(ComponentPose=@n6) {
    overridePhysicsAsset=&/M_MED_TennisLeash_Pack_Physics
    alphaBoolBlend {blendOption=Linear}
    componentAppliedLinearAccClamp=<3000 3000 3000>
    componentLinearAccScale {z=0.8}
    componentLinearVelScale {z=0.8}
    simSpaceSettings {maxAngularAcceleration=3000, maxAngularVelocity=3000, maxLinearAcceleration=3000, maxLinearVelocity=3000}
   }
   Output Pose(Result=Component To Local(ComponentPose=@n7)) {layerGroup=None}
  ]}
 EventGraph {nodes [
   Event BlueprintUpdateAnimation()
   n1 = GetPartSkeletalMeshComponent()
   > Set Transform Pelvis(Transform Pelvis:Transform=GetSocketTransform(
     self=GetAttachParent!(self:Scene Component Object Reference=@n1) -> (ReturnValue:Scene Component Object Reference)
     InSocketName=pelvis
     TransformSpace=RTS_World
    )) -> (Output_Get:Transform)
   TryGetPawnOwner()
  ]}
]
variables [Transform Pelvis {type=Transform, default {location=<0 0 0>, rotation=<rot 0 0 0>, scale=<1 1 1>}}]
